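# Material function that layers a three-tint cloth sock onto the input material for the VisionHorseGown shoe's sock inside.
# An RGB mask, tiled and rotated separately per foot via a UV V step, drives color, specular/roughness/metallic/fuzz, emissive and detail-normal flatness.
# Cloth shading adds textured Fresnel fuzz, and a channel-selected mask input blends the result.
# A HideSocks static switch collapses vertices to object position through WPO.
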MaterialFunction /CosmeticShoes/Assets/FORT_Shoes/Shoe_VisionHorseGown/Materials/MF_Sock_VisionHorseGownInside
graph {nodes [
  n1 = NamedRerouteUsage()
  FunctionOutput(None=BlendMaterialAttributes(A=NamedRerouteUsage(), B=NamedRerouteUsage(), Alpha=Saturate(None=@n1)))
  MaterialFunctionCall(
   X=ScalarParameter() {group=Sock, parameterName=SockMaskOffsetX, sortPriority=1}
   Y=ScalarParameter() {group=Sock, parameterName=SockMaskOffsetY, sortPriority=1}
  ) {materialFunction=&MakeFloat2}
  NamedRerouteDeclaration(None=FunctionInput() {inputName=Material, inputType=FunctionInput_MaterialAttributes}) {name=IN_Material, nodeColor=<rgba 0 1 0.3203 1>}
  n2 = MaterialFunctionCall(Float3=NamedRerouteUsage()) {materialFunction=&BreakOutFloat3Components}
  n3 = SetMaterialAttributes!(
   MaterialAttributes=NamedRerouteUsage()
   BaseColor=NamedRerouteUsage()
   Specular=@n2.R
   Roughness=@n2.R
   Metallic=@n2.R
   Normal=NamedRerouteUsage()
   EmissiveColor=NamedRerouteUsage()
   WorldPositionOffset=NamedRerouteUsage()
  ) {attributeSetTypes [
    #1
    #2
    #3
    #4
    #5
    #6
    #7
   ]}
  n4 = SetMaterialAttributes!(
   MaterialAttributes=@n3
   ShadingModel=ShadingModel() {shadingModel=MSM_Cloth}
   BaseColor=NamedRerouteUsage()
   CustomData0=NamedRerouteUsage()
   SubsurfaceColor=NamedRerouteUsage()
   Opacity=Constant() {r=1}
  ) {attributeSetTypes [
    #8
    #1
    #9
    #10
    #11
   ]}
  NamedRerouteDeclaration(None=@n4) {name=OUT_Material, nodeColor=<rgba 0 0.1562 1 1>}
  n5 = ObjectPositionWS()
  n6 = WorldPosition()
  NamedRerouteDeclaration(None=StaticSwitchParameter(True=Subtract(A=@n5, B=@n6.XYZ), False=Constant()) {group=Sock, parameterName=HideSocks, sortPriority=-100}) {name=OUT_WPO, nodeColor=<rgba 1 0.4219 0 1>}
  n7 = LinearInterpolate(
   A=Multiply(A=ScalarParameter() {group=Sock, parameterName=SockPrimaryNormalFlatness, sortPriority=40}, B=NamedRerouteUsage())
   B=ScalarParameter() {group=Sock, parameterName=SockSecondaryNormalFlatness, sortPriority=41}
   Alpha=NamedRerouteUsage()
  )
  n8 = ScalarParameter() {group=Sock, parameterName=SockTertiaryNormalFlatness, sortPriority=42}
  n9 = NamedRerouteUsage()
  NamedRerouteDeclaration(None=MaterialFunctionCall(Normal=NamedRerouteUsage(), Flatness=LinearInterpolate(A=@n7, B=@n8, Alpha=@n9)) {materialFunction=&FlattenNormal}.Result) {name=OUT_Normal, nodeColor=<rgba 0.6172 0 1 1>}
  n10 = Multiply(
   A=Multiply(
    A=VectorParameter() {group=Sock, parameterName=SockPrimaryEmissive, sortPriority=30, defaultValue=<rgba 0 0 0 1>}.RGB
    B=ScalarParameter() {defaultValue=1, group=Sock, parameterName=SockPrimaryEmissiveStrength, sortPriority=31}
   )
   B=NamedRerouteUsage()
  )
  n11 = Multiply(
   A=Multiply(
    A=VectorParameter() {group=Sock, parameterName=SockSecondaryEmissive, defaultValue=<rgba 0 0 0 1>}.RGB
    B=ScalarParameter() {defaultValue=1, group=Sock, parameterName=SockSecondaryEmissiveStrength, sortPriority=33}
   )
   B=NamedRerouteUsage()
  )
  n12 = Multiply(
   A=VectorParameter() {group=Sock, parameterName=SockTertiaryEmissive, sortPriority=34, defaultValue=<rgba 0 0 0 1>}.RGB
   B=ScalarParameter() {defaultValue=1, group=Sock, parameterName=SockTertiaryEmissiveStrength, sortPriority=35}
  )
  n13 = NamedRerouteUsage()
  NamedRerouteDeclaration(None=Add(A=Add(A=@n10, B=@n11), B=Multiply(A=@n12, B=@n13))) {name=OUT_Emissive, nodeColor=<rgba 0.0781 1 0 1>}
  n14 = DotProduct(A=PixelNormalWS(), B=CameraVectorWS())
  n15 = Power(
   Base=OneMinus(None=Saturate(None=@n14))
   Exp=ScalarParameter() {defaultValue=1.5, group=Sock, parameterName=SockFuzzExponent, sortPriority=402}
  )
  n16 = ScalarParameter() {defaultValue=3.5, group=Sock, parameterName=SockFuzzTiling, sortPriority=401}
  n17 = ScalarParameter() {defaultValue=6.5283, group=Sock, parameterName=SockInsideTile}
  n18 = TextureSampleParameter2D(UVs=Multiply(A=TextureCoordinate(), B=Multiply(A=@n16, B=@n17))) {
   group=Sock
   parameterName=SockFuzzTexture
   samplerSource=SSM_Wrap_WorldGroupSettings
   samplerType=SAMPLERTYPE_LinearColor
   sortPriority=400
   texture=/Game/Global/Textures/Default/Masks_and_Gradients/T_Fuzz_MASK
  }
  n19 = Saturate(None=Multiply(A=Multiply(A=@n15, B=@n18.R), B=ScalarParameter() {defaultValue=0.5, group=Sock, parameterName=SockFuzz Fresnel Blend, sortPriority=403}))
  n20 = Multiply(A=Multiply(A=NamedRerouteUsage()) {constB=0.9}, B=Add(A=Multiply(A=@n19) {constB=5}))
  NamedRerouteDeclaration(None=LinearInterpolate(A=NamedRerouteUsage(), B=Saturate(None=@n20), Alpha=NamedRerouteUsage())) {name=Cloth_BaseColor, nodeColor=<rgba 0 1 0.7188 1>}
  n21 = Multiply(A=VectorParameter() {group=Sock, parameterName=SockPrimaryColor, sortPriority=10, defaultValue=<rgba 1 0 0 1>}.RGB, B=NamedRerouteUsage())
  n22 = Multiply(A=VectorParameter() {group=Sock, parameterName=SockSecondaryColor, sortPriority=12, defaultValue=<rgba 0 1 0 1>}.RGB, B=NamedRerouteUsage())
  n23 = VectorParameter() {group=Sock, parameterName=SockTertiaryColor, sortPriority=14, defaultValue=<rgba 0 0 1 1>}
  n24 = NamedRerouteUsage()
  NamedRerouteDeclaration(None=Add(A=Add(A=@n21, B=@n22), B=Multiply(A=@n23.RGB, B=@n24))) {name=OUT_BaseColor, nodeColor=<rgba 0 0.0391 1 1>}
  n25 = NamedRerouteUsage()
  n26 = ScalarParameter() {defaultValue=0.3, group=Sock, parameterName=SockFuzzSSIntensity, sortPriority=405}
  NamedRerouteDeclaration(None=LinearInterpolate(B=Multiply(A=@n25, B=@n26), Alpha=NamedRerouteUsage())) {name=Cloth_SubSurface, nodeColor=<rgba 0 0.6484 1 1>}
  n27 = VectorParameter() {
   group=Sock
   parameterName=SockPrimarySRMF
   sortPriority=20
   channelNames {a=Fuzz, b=Metallic, g=Roughness, r=Specular}
   defaultValue=<rgba 0.4 0.8 0 1>
  }
  n28 = AppendVector(A=@n27, B=@n27)
  n29 = NamedRerouteUsage()
  n30 = VectorParameter() {
   group=Sock
   parameterName=SockSecondarySRMF
   sortPriority=22
   channelNames {a=Fuzz, b=Metallic, g=Roughness, r=Specular}
   defaultValue=<rgba 0.4 0.8 0 1>
  }
  n31 = AppendVector(A=@n30, B=@n30)
  n32 = NamedRerouteUsage()
  n33 = VectorParameter() {
   group=Sock
   parameterName=SockTertiarySRMF
   sortPriority=24
   channelNames {a=Fuzz, b=Metallic, g=Roughness, r=Specular}
   defaultValue=<rgba 0.4 0.8 0 1>
  }
  n34 = Add(A=Add(A=Multiply(A=@n28, B=@n29), B=Multiply(A=@n31, B=@n32)), B=Multiply(A=AppendVector(A=@n33, B=@n33), B=NamedRerouteUsage()))
  NamedRerouteDeclaration(None=ComponentMask(None=@n34) {a=true}) {name=OUT_Fuzz, nodeColor=<rgba 0.2187 1 0 1>}
  NamedRerouteDeclaration(None=ChannelMaskParameter(None=FunctionInput() {bUsePreviewValueAsDefault=true, inputName=Mask, sortPriority=1, previewValue=<1 1 1 1>}) {group=Sock, parameterName=SockTechArtMaskChannel, sortPriority=-10}.R) {name=SockAlpha, nodeColor=<rgba 0.6406 1 0 1>}
  NamedRerouteDeclaration(None=Saturate(None=ComponentMask(None=@n34) {b=true, g=true, r=true})) {name=OUT_SRM, nodeColor=<rgba 0.0781 1 0 1>}
  n35 = Constant()
  n36 = ScalarParameter() {defaultValue=1, group=Sock, parameterName=SockMaskTileX, sortPriority=1}
  n37 = ScalarParameter() {defaultValue=1, group=Sock, parameterName=SockMaskTileY, sortPriority=1}
  n38 = ScalarParameter() {defaultValue=6.9235, group=Sock, parameterName=SockInsideTileLeft}
  n39 = ScalarParameter() {defaultValue=6.5283, group=Sock, parameterName=SockInsideTileRight}
  n40 = Step(X=ComponentMask(None=TextureCoordinate()) {g=true}) {constX=0.6, constY=0.6}
  n41 = Multiply(
   A=TextureCoordinate()
   B=Multiply(
    A=MaterialFunctionCall(X=@n36, Y=@n37) {materialFunction=&MakeFloat2}.Result
    B=LinearInterpolate(A=@n38, B=@n39, Alpha=@n40)
   ) {constB=1.4}
  )
  n42 = Multiply(
   A=ComponentMask(None=Constant3Vector() {constant=<rgba 0.58 0.521 0 1>}) {g=true, r=true}
   B=ScalarParameter() {defaultValue=6.9235, group=Sock, parameterName=SockInsideTileLeft}
  )
  n43 = Multiply(
   A=ComponentMask(None=Constant3Vector() {constant=<rgba 0.546 0.858 0 1>}) {g=true, r=true}
   B=ScalarParameter() {defaultValue=6.5283, group=Sock, parameterName=SockInsideTileRight}
  ) {constB=8.6}
  n44 = Step(X=ComponentMask(None=TextureCoordinate()) {g=true}) {constX=0.6, constY=0.6}
  n45 = TextureCoordinate()
  n46 = LinearInterpolate(
   A=ScalarParameter() {defaultValue=0.26, parameterName=SockInsideRotationLeft}
   B=ScalarParameter() {defaultValue=-0.13, group=Sock, parameterName=SockInsideRotationRight}
   Alpha=Step(X=ComponentMask(None=@n45) {g=true}) {constX=0.6, constY=0.6}
  )
  n47 = TextureSampleParameter2D(UVs=MaterialFunctionCall(
    UVs=Add(A=@n35, B=@n41)
    Rotation Center=LinearInterpolate(A=@n42, B=@n43, Alpha=@n44)
    "Rotation Angle (0-1)"=NamedRerouteDeclaration(None=@n46) {name=SockInsideRotation, nodeColor=<rgba 1 0.3047 0 1>}
   ) {materialFunction=&CustomRotator}.Rotated Values) {
   group=Sock
   parameterName=SockMaskTexture
   sortPriority=0
   texture=/CosmeticShoes/Assets/FORT_Shoes/Shoe_VisionHorseGown/Textures/SockTexture/T_Cobbler_Socks_Red_Mask
  }
  NamedRerouteDeclaration(None=@n47.R) {name=SockMask_Primary, nodeColor=<rgba 0 1 0.1562 1>}
  NamedRerouteDeclaration(None=@n47.G) {name=SockMask_Secondary, nodeColor=<rgba 0 1 0.1562 1>}
  NamedRerouteDeclaration(None=@n47.B) {name=SockMask_Tertiary, nodeColor=<rgba 0 1 0.1562 1>}
  n48 = TextureCoordinate()
  n49 = Multiply(
   A=ScalarParameter() {defaultValue=8, group=Sock, parameterName=SockDetailNormalTile, sortPriority=301}
   B=ScalarParameter() {defaultValue=8, group=Sock, parameterName=SockInsideTileN}
  ) {constB=1.4}
  n50 = TextureSampleParameter2D(UVs=MaterialFunctionCall(UVs=Multiply(A=@n48, B=@n49), "Rotation Angle (0-1)"=NamedRerouteUsage()) {materialFunction=&CustomRotator}.Rotated Values) {
   group=Sock
   parameterName=SockDetailNormal
   samplerType=SAMPLERTYPE_Normal
   sortPriority=300
   texture=/CosmeticShoes/Assets/FORT_Shoes/Shoe_VisionHorseGown/Textures/SockTexture/T_CottonJersey_N
  }
  NamedRerouteDeclaration(None=@n50.RGB) {name=IN_Normal, nodeColor=<rgba 0 0.7422 1 1>}
 ]}
