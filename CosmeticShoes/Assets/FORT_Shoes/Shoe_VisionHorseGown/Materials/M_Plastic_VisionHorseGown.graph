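# Translucent, lit plastic material for the VisionHorseGown cosmetic shoe, built on material attributes.
# A TechArtMask texture feeds the shoe's plastic function and sets the alpha for a Fresnel-driven thin-film layer.
# Static switches add optional vehicle camera dither, then choose between VertexHide, which uses world position offset to pull vertices toward the object position, and shared character opacity.
# TechArt overrides and the shared sock function finish the chain.
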Material /CosmeticShoes/Assets/FORT_Shoes/Shoe_VisionHorseGown/Materials/M_Plastic_VisionHorseGown
graph {
 properties {
  bEnableResponsiveAA=true
  bScreenSpaceReflections=true
  bUseMaterialAttributes=true
  bUsedWithClothing=true
  bUsedWithMeshDeformer=true
  bUsedWithMorphTargets=true
  bUsedWithNanite=true
  bUsedWithNiagaraMeshParticles=true
  bUsedWithSkeletalMesh=true
  blendMode=BLEND_Translucent
  ditherOpacityMask=true
  pixelDepthOffsetMode=PDOM_Legacy
  translucencyDirectionalLightingIntensity=5
  translucencyLightingMode=TLM_SurfacePerPixelLighting
  usedShadingModels=Default Lit
 }
 nodes [
  n1 = TextureSampleParameter2D() {
   group=00_Glass
   parameterName=TechArtMask
   samplerType=SAMPLERTYPE_Masks
   sortPriority=-96
   texture=&T_BlackMask
  }
  n2 = MaterialFunctionCall!(
   Material Attributes=MaterialFunctionCall!(TechArtMask=@n1.RGB) {materialFunction=&/MF_Plastic_VisionHorseGown}.Result
   Fresnel=Fresnel() {baseReflectFraction=0, exponent=1}
   Alpha=@n1.RGB
  ) {materialFunction=/Game/Global/MaterialFunctions/TechArt/MF_ThinFilm_TechArt}
  n3 = StaticSwitchParameter(
   True=MaterialFunctionCall!(In=@n2.Result) {materialFunction=/Game/Packages/Fortress_SharedMaterials/Base_Material_Functions/BaseCustomizationFunctions/MF_FN_Vehicle_CameraDither}.Result
   False=@n2.Result
  ) {group=00_Glass, parameterName=UseCameraDither, sortPriority=250}
  n4 = Subtract(A=ObjectPositionWS(), B=WorldPosition().XYZ)
  n5 = ScalarParameter() {group=00_Glass, parameterName=VertexHide, sortPriority=301}
  n6 = StaticSwitchParameter(
   True=SetMaterialAttributes!(MaterialAttributes=@n3, World Position Offset=Multiply(A=@n4, B=@n5)) {attributeSetTypes [#1]}
   False=MaterialFunctionCall!(Material=@n3) {materialFunction=/Game/Global/Materials/Character_Parent_2023/Functions/BaseFN_CharacterOpacity_2023}.Result
  ) {group=00_Glass, parameterName=UseVertexHide, sortPriority=300}
  MaterialFunctionCall!(
   InOpacity=GetMaterialAttributes(None=@n6) {attributeGetTypes [#2]}.Opacity
   Material=MaterialFunctionCall(In=@n6) {materialFunction=&MF_TechArtOverrides}.Result
   Mask
  ) {materialFunction=/Game/Global/Materials/Cosmetics/Kicks/Socks/Materials/MF_Sock}
 ]
}
