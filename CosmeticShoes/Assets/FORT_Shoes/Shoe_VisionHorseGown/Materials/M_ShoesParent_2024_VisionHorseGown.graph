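# Masked parent material for the VisionHorseGown cosmetic shoe and sock, built on the 2023 character parent PreFX and PostFX functions.
# PreFX output passes through fluffy fabric, masked by Tech Art Mask and a UV U-coordinate step, then shoe reactivity and a shoe-specific sock function.
# Opacity is the max of that function's SockAlpha and the inverted mask-texture red channel; green masks the sock.
# PostFX, shared sock and hologram functions finish the chain.
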Material /CosmeticShoes/Assets/FORT_Shoes/Shoe_VisionHorseGown/Materials/M_ShoesParent_2024_VisionHorseGown
graph {
 properties {
  bScreenSpaceReflections=true
  bUseMaterialAttributes=true
  bUsedWithClothing=true
  bUsedWithMeshDeformer=true
  bUsedWithMorphTargets=true
  bUsedWithNanite=true
  bUsedWithNiagaraMeshParticles=true
  bUsedWithSkeletalMesh=true
  blendMode=BLEND_Masked
  pixelDepthOffsetMode=PDOM_Legacy
  shadingModel=MSM_FromMaterialExpression
  translucencyDirectionalLightingIntensity=5
  translucencyLightingMode=TLM_SurfacePerPixelLighting
  usedShadingModels=Default Lit | Subsurface | Clear Coat | Subsurface Profile | Cloth
 }
 nodes [
  n1 = MaterialFunctionCall() {materialFunction=&MF_CharacterParent_2023_PreFX}
  n2 = TextureCoordinate()
  n3 = Multiply(A=@n1.Tech Art Mask, B=Step(Y=ComponentMask(None=@n2) {r=true}))
  n4 = TextureSampleParameter2D() {
   group=Sock
   parameterName=Sock Transparent mask
   samplerSource=SSM_Clamp_WorldGroupSettings
   samplerType=SAMPLERTYPE_Masks
   texture=/CosmeticShoes/Assets/FORT_Shoes/Shoe_VisionHorseGown/Textures/T_VisionHorseGown_Mask
  }
  n5 = MaterialFunctionCall!(
   Material=MaterialFunctionCall!(Material=MaterialFunctionCall!(Material=@n1.Result, M=@n1.Result, TechArtMask=@n3) {materialFunction=/BRCosmetics/Accessories/FORT_Backpacks/Backpack_M_MED_PearlTote/Materials/MF_FluffyFabric}.Result) {materialFunction=/Game/Global/Materials/Cosmetics/Kicks/Shoes/Materials/MF_ShoeReactivityTest}.Result
   Mask=@n4.G
  ) {materialFunction=&/MF_Sock_VisionHorseGown}
  n6 = Max(A=@n5.SockAlpha, B=OneMinus(None=@n4.R))
  n7 = SetMaterialAttributes!(MaterialAttributes=@n5.Result, Opacity Mask=@n6) {attributeSetTypes [#1]}
  MaterialFunctionCall!(In=MaterialFunctionCall!(
    InOpacity=@n6
    Material=MaterialFunctionCall(Material=@n7, M=@n1.M, Pre Skinned Local Position=@n1.M, Distance Blend=@n1.M, Fresnel=@n1.M) {materialFunction=&MF_CharacterParent_2023_PostFX}.Result
    Mask
   ) {materialFunction=/Game/Global/Materials/Cosmetics/Kicks/Socks/Materials/MF_Sock}.Result) {materialFunction=/Game/Global/Materials/Cosmetics/Kicks/Shoes/Materials/MF_TechartShoeHologram}
 ]
}
